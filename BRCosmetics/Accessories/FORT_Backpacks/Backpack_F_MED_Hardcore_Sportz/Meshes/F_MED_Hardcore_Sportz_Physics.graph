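# Physics asset for the Hardcore Sportz backpack cosmetic, defining bodies and joints for its dangling parts.
# A box body on the pack bone anchors finger-named capsules (index, middle, ring, pinky), two flap boxes and a zipper box.
# Each of these joins to the pack with tight twist and moderate swing limits.
# An extra constraint ties the front flap to the mid flap with free swing and twist, coupling them without restricting relative rotation.
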PhysicsAsset /BRCosmetics/Accessories/FORT_Backpacks/Backpack_F_MED_Hardcore_Sportz/Meshes/F_MED_Hardcore_Sportz_Physics
bodies {
 dyn_flap_front [{shapeType=Box, extentX=6.8941, extentY=6.4719, extentZ=2.4707, center=<2.2215 0.0776 0.35>}]
 dyn_flap_mid [{shapeType=Box, extentX=5.8659, extentY=7.2015, extentZ=2.3725, center=<2.4451 0.5033 0.5172>}]
 dyn_index [{shapeType=Capsule, length=5.041, radius=2.1798, center=<3.3684 0 0>, rotation=<rot 90 0 0>}]
 dyn_middle [{shapeType=Capsule, length=4.6299, radius=2.0294, center=<3 0 0>, rotation=<rot 90 0 0>}]
 dyn_pinky [{shapeType=Capsule, length=3.5086, radius=1.7078, center=<3 0 0>, rotation=<rot 90 0 0>}]
 dyn_ring [{shapeType=Capsule, length=4.4235, radius=2.1044, center=<3.4327 0 0>, rotation=<rot 90 0 0>}]
 dyn_zipper [{shapeType=Box, extentX=2.9371, extentY=1.2267, extentZ=0.4594, center=<1.0281 0.0433 0.0361>}]
 pack [{shapeType=Box, extentX=32.2764, extentY=14.4367, extentZ=15.02, center=<1.7364 -6.5046 -5.271>, rotation=<rot 27.3954 0.1228 -16.8657>}]
}
constraints [
 {bone1Name=dyn_pinky, bone2Name=pack, swing1LimitDegrees=20, swing2LimitDegrees=30, twistLimitDegrees=5}
 {bone1Name=dyn_ring, bone2Name=pack, swing1LimitDegrees=20, swing2LimitDegrees=30, twistLimitDegrees=5}
 {bone1Name=dyn_middle, bone2Name=pack, swing1LimitDegrees=20, swing2LimitDegrees=30, twistLimitDegrees=5}
 {bone1Name=dyn_index, bone2Name=pack, swing1LimitDegrees=20, swing2LimitDegrees=30, twistLimitDegrees=5}
 {bone1Name=dyn_flap_front, bone2Name=pack, swing1LimitDegrees=20, swing2LimitDegrees=20, twistLimitDegrees=5}
 {bone1Name=dyn_flap_mid, bone2Name=pack, swing1LimitDegrees=20, swing2LimitDegrees=20, twistLimitDegrees=5}
 {bone1Name=dyn_zipper, bone2Name=pack, swing1LimitDegrees=30, swing2LimitDegrees=30, twistLimitDegrees=5}
 {
  bone1Name=dyn_flap_front
  bone2Name=dyn_flap_mid
  swing1LimitDegrees=45
  swing1Motion=Free
  swing2LimitDegrees=45
  swing2Motion=Free
  twistLimitDegrees=45
  twistMotion=Free
 }
]
